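# Material function for the CosmoHound Meteor companion that recolors base color on input material attributes, behind a Use Color Customization switch.
# Masks from Custom Color Mask channels layer primary, secondary and tertiary colors, each with an optional blend color under UseAdvancedColorBlend.
# An optional pattern mask overlays another color.
# The function also derives an adjustable value from incoming base color, declares a Tech Art Mask input, and writes the result back through SetMaterialAttributes.
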MaterialFunction /CosmeticCompanions/Assets/Quadruped/CosmoHound_Meteor/Materials/MF_FN_CosmoHound_Custom_Color
graph {nodes [
  Constant3Vector() {constant=<rgba 0.1 0 0 0>}
  Constant()
  Constant() {r=1}
  Constant() {r=0.5}
  n1 = NamedRerouteUsage()
  n2 = NamedRerouteUsage()
  FunctionOutput(None=StaticSwitchParameter(
    True=SetMaterialAttributes!(MaterialAttributes=@n1, BaseColor=@n2) {attributeSetTypes [#1]}
    False=NamedRerouteUsage()
   ) {group=Color Customization, parameterName=Use Color Customization, sortPriority=0}) {outputName=Base Color}
  n3 = TextureSampleParameter2D() {
   group=Color Customization
   parameterName=Custom Color Mask
   samplerSource=SSM_Wrap_WorldGroupSettings
   samplerType=SAMPLERTYPE_Masks
   sortPriority=11
   texture=&T_BlackMask
  }
  NamedRerouteDeclaration(None=ChannelMaskParameter(None=@n3.RGB) {group=Color Customization, parameterName=PrimaryMaskChannel, sortPriority=12}.R) {name=PrimaryMask, nodeColor=<rgba 1 0.0596 0.1687 1>}
  n4 = ChannelMaskParameter(None=@n3.RGB) {group=Color Customization, maskChannel=Green, parameterName=SecondaryMaskChannel, sortPriority=13, defaultValue=<rgba 0 1 0 0>}
  NamedRerouteDeclaration(None=@n4.R) {name=SecondaryMask, nodeColor=<rgba 0 1 0.0283 1>}
  n5 = FunctionInput() {inputName=Material, inputType=FunctionInput_MaterialAttributes}
  n6 = NamedRerouteDeclaration(None=GetMaterialAttributes(None=NamedRerouteDeclaration(None=@n5) {name=IN_Material, nodeColor=<rgba 1 0.0703 0 1>}) {attributeGetTypes [#1]}.BaseColor) {name=IN_BaseColor, nodeColor=<rgba 0 0.25 1 1>}
  NamedRerouteDeclaration(None=ChannelMaskParameter(None=@n6) {group=Color Customization, parameterName=BaseColorAdvancedBlendChannel, sortPriority=18}.R) {name=AdvancedBlendMask, nodeColor=<rgba 0.5703 1 0 1>}
  n7 = ChannelMaskParameter(None=@n6) {group=Color Customization, maskChannel=Green, parameterName=BaseColorValueChannel, sortPriority=15, defaultValue=<rgba 0 1 0 0>}
  n8 = Add(A=@n7.R, B=ScalarParameter() {group=Color Customization, parameterName=BaseColorValueBias, sortPriority=16})
  NamedRerouteDeclaration(None=Power(Base=Saturate(None=@n8), Exp=ScalarParameter() {defaultValue=1, group=Color Customization, parameterName=BaseColorValuePow, sortPriority=17})) {name=BaseColorValue, nodeColor=<rgba 1 0 0.8906 1>}
  n9 = VectorParameter() {group=Color Customization, parameterName=Primary Color, defaultValue=<rgba 1 1 1 1>}
  n10 = VectorParameter() {group=Color Customization, parameterName=Primary Blend Color, sortPriority=33, defaultValue=<rgba 1 1 1 1>}
  n11 = NamedRerouteUsage()
  NamedRerouteDeclaration(None=StaticSwitch(True=LinearInterpolate(A=@n9.RGB, B=@n10.RGB, Alpha=@n11), False=@n9.RGB, Value=NamedRerouteUsage())) {name=PrimaryColor, nodeColor=<rgba 0 1 0.7891 1>}
  n12 = VectorParameter() {group=Color Customization, parameterName=Secondary Color, sortPriority=34, defaultValue=<rgba 1 1 1 1>}
  n13 = VectorParameter() {group=Color Customization, parameterName=Secondary Blend Color, sortPriority=35, defaultValue=<rgba 1 1 1 1>}
  n14 = NamedRerouteUsage()
  NamedRerouteDeclaration(None=StaticSwitch(True=LinearInterpolate(A=@n12.RGB, B=@n13.RGB, Alpha=@n14), False=@n12.RGB, Value=NamedRerouteUsage())) {name=SecondaryColor, nodeColor=<rgba 0 1 0.7891 1>}
  n15 = VectorParameter() {group=Color Customization, parameterName=Tertiary Color, sortPriority=36, defaultValue=<rgba 1 1 1 1>}
  n16 = VectorParameter() {group=Color Customization, parameterName=Tertiary Blend Color, sortPriority=37, defaultValue=<rgba 1 1 1 1>}
  n17 = NamedRerouteUsage()
  NamedRerouteDeclaration(None=StaticSwitch(True=LinearInterpolate(A=@n15.RGB, B=@n16.RGB, Alpha=@n17), False=@n15.RGB, Value=NamedRerouteUsage())) {name=TertiaryColor, nodeColor=<rgba 0 1 0.7891 1>}
  n18 = TextureSampleParameter2D() {
   group=Color Customization
   parameterName=Pattern Mask
   samplerSource=SSM_Wrap_WorldGroupSettings
   samplerType=SAMPLERTYPE_Masks
   sortPriority=51
   texture=&T_BlackMask
  }
  n19 = VectorParameter() {
   desc=Put 1.0 in the channel that that the pattern mask is in and 0.0 in the other channels
   group=Color Customization
   parameterName=Pattern Mask Channel
   sortPriority=52
   defaultValue=<rgba 1 0 0 1>
  }
  NamedRerouteDeclaration(None=Saturate(None=DotProduct(A=@n18.RGB, B=@n19.RGB))) {name=PatternMask, nodeColor=<rgba 0.1719 1 0 1>}
  n20 = NamedRerouteUsage()
  n21 = NamedRerouteUsage()
  n22 = NamedRerouteUsage()
  n23 = NamedRerouteUsage()
  n24 = LinearInterpolate(
   A=LinearInterpolate(A=LinearInterpolate(A=@n21, B=@n22, Alpha=@n23) {constA=1}, B=NamedRerouteUsage(), Alpha=NamedRerouteUsage())
   B=NamedRerouteUsage()
   Alpha=NamedRerouteUsage()
  )
  n25 = VectorParameter() {group=Color Customization, parameterName=Pattern Color, sortPriority=53, defaultValue=<rgba 1 1 1 1>}
  n26 = NamedRerouteUsage()
  n27 = ScalarParameter() {group=Color Customization, parameterName=Pattern Strength, sortPriority=54}
  n28 = StaticSwitchParameter(True=LinearInterpolate(A=@n24, B=@n25.RGB, Alpha=Multiply(A=@n26, B=@n27)) {constA=1}, False=@n24) {group=Color Customization, parameterName=Use Pattern Mask, sortPriority=50}
  n29 = NamedRerouteUsage()
  NamedRerouteDeclaration(None=LinearInterpolate(A=NamedRerouteUsage(), B=Multiply(A=@n20, B=@n28), Alpha=Saturate(None=@n29))) {name=OUT_BaseColor, nodeColor=<rgba 1 0 0.4922 1>}
  NamedRerouteDeclaration(None=ChannelMaskParameter(None=FunctionInput() {bUsePreviewValueAsDefault=true, inputName=Tech Art Mask}) {group=Color Customization, parameterName=Tech Art Mask Channel, sortPriority=10}.R) {name=Mask, nodeColor=<rgba 0 0.8125 1 1>}
  n30 = ChannelMaskParameter(None=@n3.RGB) {group=Color Customization, maskChannel=Blue, parameterName=TertiaryMaskChannel, sortPriority=14, defaultValue=<rgba 0 0 1 0>}
  NamedRerouteDeclaration(None=@n30.R) {name=TertiaryMask, nodeColor=<rgba 0 0.2243 1 1>}
  n31 = TextureSampleParameter2D() {
   group=Color Customization
   parameterName=ColorBlendMask
   samplerType=SAMPLERTYPE_Masks
   sortPriority=11
   texture=/CosmeticCompanions/Assets/Quadruped/CosmoHound/Textures/T_CosmoHound_Body_FX
  }
  n32 = ChannelMaskParameter(None=@n31.RGB) {
   desc="You can set R=1, G=1, and B=1 if you want to mix all of the channels together.  You aren't limited to just R, G, or B."
   group=Color Customization
   maskChannel=Green
   parameterName=Color Blend Mask Channel
   sortPriority=10
   defaultValue=<rgba 0 1 0 0>
  }
  NamedRerouteDeclaration(None=@n32.R) {name=ColorBlendMask, nodeColor=<rgba 1 0.9375 0 1>}
  NamedRerouteDeclaration(None=StaticBoolParameter() {group=Color Customization, parameterName=UseAdvancedColorBlend, sortPriority=20}) {name=bAdvancedColorBlend, nodeColor=<rgba 0.7109 0 1 1>}
  n33 = VectorParameter() {group=Color Customization, parameterName=Base Color, sortPriority=30, defaultValue=<rgba 1 1 1 1>}
  n34 = VectorParameter() {group=Color Customization, parameterName=Base Blend Color, sortPriority=31, defaultValue=<rgba 1 1 1 1>}
  n35 = NamedRerouteUsage()
  NamedRerouteDeclaration(None=StaticSwitch(True=LinearInterpolate(A=@n33.RGB, B=@n34.RGB, Alpha=@n35), False=@n33.RGB, Value=NamedRerouteUsage())) {name=BaseColor, nodeColor=<rgba 0 1 0.3906 1>}
 ]}
