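# Character material for the CosmoHound Meteor quadruped cosmetic, likely an accessory variant, built on the 2023 CharacterParent PreFX and PostFX functions.
# PreFX output passes through a CosmoHound custom color function, then BaseColorMask blends base color toward black.
# Emissive blends in a tint by channel mask, darkens edges with inverted Fresnel (labeled for eyes), then remaps through SmoothStep before PostFX.
# An emissive brightness modulation switches its multiplier by shading path for mobile.
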Material /CosmeticCompanions/Assets/Quadruped/CosmoHound_Meteor/Materials/M_CosmoHound_Meteor_Acc
graph {
 properties {
  bAutomaticallySetUsageInEditor=false
  bUseMaterialAttributes=true
  bUsedWithMeshDeformer=true
  bUsedWithMorphTargets=true
  bUsedWithSkeletalMesh=true
  bUsedWithStaticLighting=true
  previewMesh=/CosmeticCompanions/Assets/Quadruped/CosmoHound/Meshes/CosmoHound_Quadruped
  shadingModel=MSM_FromMaterialExpression
  usedShadingModels=Default Lit | Subsurface | Clear Coat | Subsurface Profile | Cloth
 }
 nodes [
  n1 = MaterialFunctionCall() {materialFunction=&MF_CharacterParent_2023_PreFX}
  n2 = GetMaterialAttributes(None=MaterialFunctionCall!(Material=@n1.Result, Tech Art Mask=NamedRerouteDeclaration(None=@n1.Tech Art Mask) {name=TechArtMask, nodeColor=<rgba 1 0 0.4453 1>}) {materialFunction=&/MF_FN_CosmoHound_Custom_Color}.Base Color) {attributeGetTypes [#1, #2]}
  n3 = ChannelMaskParameter(None=NamedRerouteUsage()) {group=Color Customization, parameterName=BaseColorMask}
  n4 = ScalarParameter() {parameterName=MinA}
  n5 = ScalarParameter() {defaultValue=1, parameterName=MaxA}
  n6 = VectorParameter() {parameterName=EmissiveTint, defaultValue=<rgba 1 1 1 1>}
  n7 = NamedRerouteUsage()
  n8 = NamedRerouteUsage()
  n9 = ScalarParameter() {defaultValue=0.8, parameterName=EdgeOfEyeDarkeningFresenelExponent}
  n10 = Multiply(
   A=LinearInterpolate(A=NamedRerouteUsage(), B=Multiply(A=@n6.RGB, B=@n7), Alpha=ChannelMaskParameter(None=@n8) {parameterName=EmissiveMaskChannel}.R)
   B=OneMinus(None=Fresnel(ExponentIn=@n9) {exponent=0.5})
  ) {constB=2}
  MaterialFunctionCall(
   Material=SetMaterialAttributes!(
    MaterialAttributes=@n2.MaterialAttributes
    Base Color=LinearInterpolate(A=@n2.Base Color, Alpha=@n3.R) {constB=0}
    Emissive Color=SmoothStep(Min=@n4, Max=@n5, Value=@n10) {constMin=0.168}
   ) {attributeSetTypes [#1, #2]}
   M=@n1.M
   Pre Skinned Local Position=@n1.M
   Distance Blend=@n1.M
   Fresnel=@n1.M
  ) {materialFunction=&MF_CharacterParent_2023_PostFX}
  NamedRerouteDeclaration(None=@n2.Emissive Color) {name=In_EmissiveColor, nodeColor=<rgba 0 1 0.3203 1>}
  n11 = ScalarParameter() {defaultValue=1, group=06_Emissive, parameterName=EmissiveMobileMultiplier}
  n12 = Constant() {r=1}
  NamedRerouteDeclaration(None=Multiply(A=ScalarParameter() {defaultValue=1, group=06_Emissive, parameterName=EmissiveBrightness}, B=ShadingPathSwitch(Default=@n11, Deferred=@n12))) {name=Emissive Brightness Modulation, nodeColor=<rgba 1 0.3281 0 1>}
 ]
}
