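# Material function for the CosmoHound Meteor companion that adds detail to base color from an extra texture, exposed as a face texture parameter.
# It samples the texture on UV1, splitting red into a tech-art mask and green into a hex grid scaled by a multiplier.
# An inverted Fresnel factor limits the hex-grid blend to camera-facing surfaces.
# Other material attributes, including emissive, pass through unchanged.
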MaterialFunction /CosmeticCompanions/Assets/Quadruped/CosmoHound_Meteor/Materials/MF_CosmoHound_ExtraTexture
graph {nodes [
  n1 = GetMaterialAttributes(None=FunctionInput() {inputName=Material, inputType=FunctionInput_MaterialAttributes}) {attributeGetTypes [#1, #2]}
  n2 = NamedRerouteUsage()
  n3 = TextureObjectParameter() {parameterName=ExtraTextureForFace, texture=/CosmeticCompanions/Assets/Quadruped/CosmoHound_Meteor/Textures/T_CosmoHound_Meteor_ExtraTexture}
  n4 = TextureSample(Tex=@n3) {constCoordinate=1}
  n5 = NamedRerouteDeclaration(None=@n4.G) {name=HexGrid, nodeColor=<rgba 1 0.5156 0 1>}
  n6 = ScalarParameter() {defaultValue=3, parameterName=ExtraTextureMultiplier}
  n7 = NamedRerouteDeclaration(None=Multiply(A=NamedRerouteUsage(), B=Multiply(A=@n5, B=@n6) {constB=2})) {name=HexGridBaseColor, nodeColor=<rgba 0 1 0.3672 1>}
  n8 = Fresnel() {exponent=0.33}
  n9 = Multiply(A=NamedRerouteUsage(), B=OneMinus(None=Saturate(None=@n8)))
  FunctionOutput(None=SetMaterialAttributes!(
    MaterialAttributes=@n1.MaterialAttributes
    EmissiveColor=NamedRerouteDeclaration(None=@n1.EmissiveColor) {name=In_EmissiveColor, nodeColor=<rgba 1 0 0.8437 1>}
    BaseColor=LinearInterpolate(A=@n2, B=@n7, Alpha=@n9)
   ) {attributeSetTypes [#1, #2]}) {outputName=MaterialAttributes}
  NamedRerouteDeclaration(None=@n1.BaseColor) {name=In_BaseColor, nodeColor=<rgba 0 1 0.8359 1>}
  NamedRerouteDeclaration(None=@n4.R) {name=TechArtMask, nodeColor=<rgba 1 0.0469 0 1>}
 ]}
