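# Body material for the CosmoHound Meteor quadruped companion, with an animated flipbook face on a visor.
# It chains the shared character PreFX, extra-texture, custom-color, face and visor functions into PostFX, with visor UVs driving eye placement.
# Eye tint switches to BoneEyeTint when the animation-driven extra-pose curve passes a threshold.
# Emissive blends non-eye glow with eye color, adds Fresnel edge darkening and smoothsteps it, while an emissive mask blacks out base color.
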Material /CosmeticCompanions/Assets/Quadruped/CosmoHound_Meteor/Materials/M_CosmoHound_Meteor_Body
graph {
 properties {
  bAutomaticallySetUsageInEditor=false
  bEnableNewHLSLGeneratorOverridden=true
  bUseMaterialAttributes=true
  bUsedWithMeshDeformer=true
  bUsedWithMorphTargets=true
  bUsedWithSkeletalMesh=true
  bUsedWithStaticLighting=true
  shadingModel=MSM_FromMaterialExpression
  usedShadingModels=Default Lit | Subsurface | Clear Coat | Subsurface Profile | Cloth
 }
 nodes [
  n1 = MaterialFunctionCall() {materialFunction=&MF_CharacterParent_2023_PreFX}
  n2 = GetMaterialAttributes(None=MaterialFunctionCall!(
    Material=MaterialFunctionCall!(Material=@n1.Result) {materialFunction=&/MF_CosmoHound_ExtraTexture}.MaterialAttributes
    Tech Art Mask=NamedRerouteDeclaration(None=@n1.Tech Art Mask) {name=TechArtMask, nodeColor=<rgba 1 0 0.4453 1>}
   ) {materialFunction=&/MF_FN_CosmoHound_Custom_Color}.Base Color) {attributeGetTypes [#1]}
  n3 = MaterialFunctionCall!(Material=@n2.MaterialAttributes, Brow UV, Eye UV=@n4.UVs, Mouth UV) {materialFunction=/CosmeticCompanions/Materials/MaterialFunctions/MF_FlipbookFace_Companion}
  n4 = MaterialFunctionCall!(Material=@n3.Material, Tech Art Mask=@n1.Tech Art Mask, In_TintColor=NamedRerouteUsage()) {materialFunction=&/MF_CosmoHoundVisor}
  n5 = GetMaterialAttributes(None=@n4.Result) {attributeGetTypes [#1, #2]}
  n6 = ScalarParameter() {parameterName=MinA}
  n7 = ScalarParameter() {defaultValue=1, parameterName=MaxA}
  n8 = NamedRerouteUsage()
  n9 = ScalarParameter() {defaultValue=2, parameterName=NonEyeEmissiveMultiplier}
  n10 = Multiply(A=VectorParameter() {parameterName=EmissiveTint, defaultValue=<rgba 1 1 1 1>}.RGB, B=NamedRerouteUsage())
  n11 = Multiply(
   A=Multiply(A=@n3.EyesColor, B=ScalarParameter() {defaultValue=1.1, parameterName=EyeColorMultiplier})
   B=If(
    A=ScalarParameter() {desc=extra eye pose for emote, group="Eyes Curve Parameters (Driven By Animation)", parameterName=R_extra_pose}
    B=ScalarParameter() {defaultValue=0.1, group=Geometry Flipbook, parameterName=Threshold_EyeExtra}
    "A > B"=VectorParameter() {group=06_Emissive, parameterName=BoneEyeTint, defaultValue=<rgba 4 1.8396 0.1875 1>}.RGB
    "A == B"=@n10
    "A < B"=@n10
   )
  )
  n12 = ScalarParameter() {defaultValue=0.8, parameterName=EdgeOfEyeDarkeningFresenelExponent}
  n13 = Multiply(
   A=LinearInterpolate(A=Multiply(A=@n8, B=@n9), B=NamedRerouteDeclaration(None=@n11) {name=Tint, nodeColor=<rgba 1 0 0.5391 1>}, Alpha=@n3.EyesAlpha)
   B=OneMinus(None=Fresnel(ExponentIn=@n12) {exponent=0.5})
  ) {constB=2}
  n14 = NamedRerouteUsage()
  n15 = TextureSampleParameter2D() {
   parameterName=EmissiveMask
   samplerType=SAMPLERTYPE_Masks
   texture=/CosmeticCompanions/Assets/Quadruped/CosmoHound_Meteor/Textures/T_CosmoHound_Meteor_Body_FX1
  }
  n16 = NamedRerouteDeclaration(None=ChannelMaskParameter(None=@n15.RGBA) {maskChannel=Green, parameterName=EmissiveMaskChannel, defaultValue=<rgba 0 1 0 0>}.R) {name=EmissiveMask, nodeColor=<rgba 1 0.9375 0 1>}
  MaterialFunctionCall(
   Material=SetMaterialAttributes!(
    MaterialAttributes=@n5.MaterialAttributes
    Emissive Color=SmoothStep(Min=@n6, Max=@n7, Value=@n13) {constMin=0.168}
    Base Color=LinearInterpolate(A=@n14, Alpha=@n16) {constB=0}
   ) {attributeSetTypes [#1, #2]}
   M=@n1.M
   Pre Skinned Local Position=@n1.M
   Distance Blend=@n1.M
   Fresnel=@n1.M
  ) {materialFunction=&MF_CharacterParent_2023_PostFX}
  NamedRerouteDeclaration(None=@n5.Base Color) {name=In_BaseColor, nodeColor=<rgba 0.125 1 0 1>}
  n17 = ScalarParameter() {defaultValue=1, group=06_Emissive, parameterName=EmissiveMobileMultiplier}
  n18 = Constant() {r=1}
  NamedRerouteDeclaration(None=Multiply(A=ScalarParameter() {defaultValue=1, group=06_Emissive, parameterName=EmissiveBrightness}, B=ShadingPathSwitch(Default=@n17, Deferred=@n18))) {name=Emissive Brightness Modulation, nodeColor=<rgba 1 0.3281 0 1>}
  NamedRerouteDeclaration(None=@n2.Emissive Color) {name=EmissiveBeforeEyes, nodeColor=<rgba 1 0 0 1>}
  TextureCoordinate() {coordinateIndex=1}
 ]
}
